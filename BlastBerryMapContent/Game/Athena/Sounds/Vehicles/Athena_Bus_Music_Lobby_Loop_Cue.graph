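# Sound cue that loops party music for the Athena bus in the lobby, assigned to the gadgets sound class.
# A quality-level node selects between two looping wave players, with the standard wave on the first two levels and the "_opt" variant on the last two.
# The result passes through an attenuation node that reuses the close-range lobby bus engine attenuation, with overall volume slightly reduced.
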SoundCue /BlastBerryMapContent/Game/Athena/Sounds/Vehicles/Athena_Bus_Music_Lobby_Loop_Cue
root=Attenuation_5
nodes [
 {
  id=Attenuation_5
  children [QualityLevel_1]
  properties {attenuationSettings=/Game/Sounds/Attenuation/Fort_Bus_Lobby_Engine_Athena_Close_Att}
 }
 {id=QualityLevel_1, children [WavePlayer_10, WavePlayer_10, WavePlayer_2, WavePlayer_2]}
 {
  id=WavePlayer_10
  properties {bLooping=true, soundWaveAssetPtr=/BlastberryMapContent/Game/Athena/Sounds/Vehicles/SFX_Bus_Flying_PartyMusic_Loop}
 }
 {
  id=WavePlayer_2
  properties {bLooping=true, soundWaveAssetPtr=/BlastberryMapContent/Game/Athena/Sounds/Vehicles/SFX_Bus_Flying_PartyMusic_Loop_opt}
 }
]
properties {maxDistance=12500, soundClassObject=&Fort_Gadgets, volumeMultiplier=0.78}
