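# Sound cue for the looping party music of the flying Battle Bus in the BlastBerry map's Athena vehicle sounds.
# A single looping wave player plays the party music wave directly, with no mixing or randomization.
# It routes to the Fort_Gadgets sound class at slightly reduced volume and references a dedicated bus music attenuation asset for distance behavior.
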SoundCue /BlastBerryMapContent/Game/Athena/Sounds/Vehicles/Athena_Bus_Music_Loop_Cue
root=WavePlayer_10
nodes [{
  id=WavePlayer_10
  properties {bLooping=true, soundWaveAssetPtr=/BlastberryMapContent/Game/Athena/Sounds/Vehicles/SFX_Bus_Flying_PartyMusic_Loop}
 }]
properties {
 attenuationSettings=/BlastBerryMapContent/Game/Sounds/Attenuation/Fort_Bus_Music_Athena_Att
 maxDistance=21000
 soundClassObject=&Fort_Gadgets
 volumeMultiplier=0.9
 attenuationOverrides {bAttenuate=false, bSpatialize=false}
}
