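; Battle bus lobby sound controller that plays cosmetic ignition and looping audio during pre-flight warmup.
; It listens for game phase step events, and on GetReady caches sounds from the local player's battle bus asset.
; Ignition plays first, then looping engine sound and music spawn after a delay.
; Tick drives only the looping sound's volume along a fade curve until warmup countdown end, and switches off at BusFlying.
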(event EventEndPlay
  (params (EndPlayReason :type EEndPlayReason%20Enum :default "Destroyed"))
  (bind _gamephasestepupdatedasynctask (Variables|Default|GetGamePhaseStepUpdatedAsyncTask)
    (:then
      (GameplayEventRouter|StopListeningtoStatefulEvent _gamephasestepupdatedasynctask))
    (:else)))

(event EventBeginPlay
  (Networking|CanExecuteCosmeticEvents
    (:True
      (bind _returnvalue (GameplayEventRouter|GetEventRouter "/Script/FortniteGame.FortGameState" self))
      (bind (_asynctaskproxy _payload) (GameplayEventRouter|StartListeningToStatefulEvent _returnvalue "/Script/FortniteGame.FortGamePhaseStepUpdatedEvent")
        (:then
          (Variables|Default|SetGamePhaseStepUpdatedAsyncTask _asynctaskproxy))
        (:OnEventReceived
          _)
        (:OnSavedEventStateExists
          _))
      (CallFunction|OnGamePhaseStepChanged (Utilities|Struct|BreakFortGamePhaseStepUpdatedEvent _payload)))
    (:False
      (Actor|Tick|SetActorTickEnabled false))))

(event Custom|On%20Game%20Phase%20Step%20Changed
  (params (GamePhaseStep :type EAthenaGamePhaseStep%20Enum))
  (switch Utilities|FlowControl|Switch|SwitchonEAthenaGamePhaseStep GamePhaseStep
    (:GetReady
      (CallFunction|ConfigureSounds)
      (Audio|SpawnSoundAttached (Variables|Default|GetIgnitionSound) (Variables|Default|GetDefaultSceneRoot))
      (Utilities|FlowControl|Delay (Variables|Default|GetIgnitionTime))
      (Utilities|FlowControl|Sequence
        (:then_0
          (bind _returnvalue (Audio|SpawnSoundAttached (Variables|Default|GetLoopingSound) (Variables|Default|GetDefaultSceneRoot)))
          (Variables|Default|SetLoopingSound_Inst _returnvalue))
        (:then_1
          (bind _returnvalue_2 (Audio|SpawnSoundAttached (Variables|Default|GetLoopingMusic) (Variables|Default|GetDefaultSceneRoot)))
          (Variables|Default|SetLoopingMusic_Inst _returnvalue_2))
        (:then_2
          (Variables|Default|SetGetReadyGameTime (Utilities|Time|GetGameTimeinSeconds))
          (Variables|Default|SetbGettingReady true))))
    (:BusFlying
      (Actor|Tick|SetActorTickEnabled false))))

(event EventTick
  (params (DeltaSeconds :type Float%20%28single-precision%29))
  (bind _returnvalue (BattleRoyaleGamePhaseLogic|GetBattleRoyaleGamePhaseLogicComponent))
  (if (Variables|Default|GetbGettingReady)
    (Utilities|IsValid _returnvalue
      (:Is%20Valid
        (Audio|Components|Audio|SetVolumeMultiplier (Variables|Default|GetLoopingSound_Inst) (Math|Curves|GetFloatValue (Variables|Default|GetGetReadyFadeCurve) (Math|Float|MapRangeClamped (Utilities|Time|GetGameTimeinSeconds) (Variables|Default|GetGetReadyGameTime) (BattleRoyaleGamePhaseLogic|GetWarmupCountdownEndTime _returnvalue) 0.0 1)) :!class /Script/Engine.AudioComponent)))))

(fn ConfigureSounds
  (if (not (Variables|Default|GetbConfiguredSounds))
    (Variables|Default|SetbConfiguredSounds true)
    (bind _battlebus (CallFunction|GetBattleBusAsset))
    (Variables|Default|SetIgnitionSound (Athena|GetIgnitionSound _battlebus))
    (Variables|Default|SetLoopingSound (Athena|GetLobbyLoopingSound _battlebus))
    (Variables|Default|SetLoopingMusic (Athena|GetLobbyLoopingMusic _battlebus))))

(fn GetBattleBusAsset
  (bind _firstlocalpc (CallFunction|GetLocalPlayer))
  (bind _asfort_game_state_athena (Utilities|Casting|CastToFortGameStateAthena (Game|GetGameState))
    (:then
      _)
    (:CastFailed
      _))
  (return (Aircraft|GetBattleBusforPlayer _asfort_game_state_athena _firstlocalpc)))

(fn GetLocalPlayer
  (bind _returnvalue (Fortnite|GetLocalFortPlayerControllers true))
  (if (> (Utilities|Array|Length _returnvalue) 0)
    (bind _asfort_player_controller_athena (Utilities|Casting|CastToFortPlayerControllerAthena (Utilities|Array|Get(acopy) _returnvalue))
      (:then
        (return _asfort_player_controller_athena))
      (:CastFailed
        _)))
  (return 0))

(fn ConstructionScript)
